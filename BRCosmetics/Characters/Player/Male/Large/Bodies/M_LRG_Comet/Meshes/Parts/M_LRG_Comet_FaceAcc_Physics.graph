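# Physics asset for the face accessory of the large male Comet player character, giving a hat's side flaps, tassels and tail secondary motion.
# Left and right flap box chains and tassel capsule chains hang from small base bodies with free angular constraints to neck_02.
# Links below each base use tight swing and twist limits, as does the two-capsule tail attached directly to the neck.
# Compound neck_02 shapes and a spine_05 capsule represent the head and upper body.
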PhysicsAsset /BRCosmetics/Characters/Player/Male/Large/Bodies/M_LRG_Comet/Meshes/Parts/M_LRG_Comet_FaceAcc_Physics
bodies {
 dyn_hatFlap_01_Base_l [{shapeType=Box, extentX=1.01, extentY=1.01, extentZ=1.01, rotation=<rot 0 -90 0>}]
 dyn_hatFlap_01_Base_r [{shapeType=Box, extentX=1.01, extentY=1.01, extentZ=1.01, rotation=<rot 0 -90 0>}]
 dyn_hatFlap_01_l [{shapeType=Box, extentX=8.0232, extentY=7.7714, extentZ=15.5477, center=<4.7257 -0.2184 -0.4253>, rotation=<rot 2.6875 -89.9994 -9.6775>}]
 dyn_hatFlap_01_r [{shapeType=Box, extentX=8.5824, extentY=10.5176, extentZ=14.0593, center=<5.0797 0.0809 -0.6952>, rotation=<rot 3.6291 -90 17.8135>}]
 dyn_hatFlap_02_l [{shapeType=Box, extentX=7.7133, extentY=12.2555, extentZ=14.4754, center=<5.8001 -0.1868 1.5546>, rotation=<rot -5.2513 -90 -12.0664>}]
 dyn_hatFlap_02_r [{shapeType=Box, extentX=7.9306, extentY=10.8525, extentZ=14.6567, center=<4.6633 0.3129 -1.5003>, rotation=<rot 1.051 -89.9999 15.7392>}]
 dyn_hatTassle_1_Base_r [{shapeType=Box, extentX=1.01, extentY=1.01, extentZ=1.01, rotation=<rot 0 -90 0>}]
 dyn_hatTassle_1_r [{
   shapeType=Capsule
   length=5.1114
   radius=0.5948
   shapeName=dyn_hatTassle_1_r_capsule
   center=<2.2581 -0.0001 0.0344>
   rotation=<rot 89.1701 179.9994 179.9994>
  }]
 dyn_hatTassle_2_r [{
   shapeType=Capsule
   length=6.0666
   radius=0.7516
   shapeName=dyn_hatTassle_2_r_capsule
   center=<3.1227 0 0.0469>
   rotation=<rot 89.1204 -180 -180>
  }]
 dyn_hatTassle_3_r [{
   shapeType=Capsule
   length=6.5529
   radius=1.3733
   shapeName=dyn_hatTassle_3_r_capsule
   center=<3.1227 0 0.047>
   rotation=<rot 89.1379 179.9996 179.9996>
  }]
 dyn_hatTassle_5_Base_l [{shapeType=Box, extentX=1.01, extentY=1.01, extentZ=1.01, rotation=<rot 0 -90 0>}]
 dyn_hatTassle_5_l [{
   shapeType=Capsule
   length=5.1114
   radius=0.5948
   shapeName=dyn_hatTassle_5_l_capsule
   center=<2.2581 -0.0001 0.0345>
   rotation=<rot 89.1244 -179.9974 -179.9974>
  }]
 dyn_hatTassle_6_l [{
   shapeType=Capsule
   length=6.0666
   radius=0.7516
   shapeName=dyn_hatTassle_6_l_capsule
   center=<3.1227 0 0.0484>
   rotation=<rot 89.0924 -179.9993 -179.9993>
  }]
 dyn_hatTassle_7_l [{
   shapeType=Capsule
   length=6
   radius=1
   shapeName=dyn_hatTassle_7_l_capsule
   center=<3.1227 0 0.0477>
   rotation=<rot 89.1244 -179.999 -179.999>
  }]
 dyn_hattail_01 [{shapeType=Capsule, length=1, radius=1, shapeName=dyn_hattail_01_capsule, center=<0.5913 0 0>, rotation=<rot 0 0 90>}]
 dyn_hattail_02 [{shapeType=Capsule, length=1.5, radius=1.5, shapeName=dyn_hattail_02_capsule, center=<0.4731 0 0>, rotation=<rot 90 0 0>}]
 neck_02 [
  {shapeType=Sphere, radius=19.2468, center=<-1.0801 -2.8893 0.0526>}
  {shapeType=Capsule, length=19.8463, radius=13.7186, center=<-1.2059 -1.262 0>}
  {shapeType=Capsule, length=5, radius=14.5, shapeName=neck_02_capsule, center=<4.1409 0.734 0.0002>}
  {shapeType=Box, extentX=12.5, extentY=2.5, extentZ=11.5, shapeName=neck_02_box, center=<10.685 -18.5927 0.0002>, rotation=<rot 0 24.9938 0>}
 ]
 spine_05 [{shapeType=Capsule, length=20.7696, radius=16.1096, shapeName=spine_05_capsule, center=<0 -1.488 0>}]
}
constraints [
 {
  bone1Name=dyn_hatTassle_1_Base_r
  bone2Name=neck_02
  swing1LimitDegrees=45
  swing1Motion=Free
  swing2LimitDegrees=45
  swing2Motion=Free
  twistLimitDegrees=45
  twistMotion=Free
 }
 {bone1Name=dyn_hatTassle_1_r, bone2Name=dyn_hatTassle_1_Base_r, swing1LimitDegrees=20, swing2LimitDegrees=8, twistLimitDegrees=8}
 {bone1Name=dyn_hatTassle_2_r, bone2Name=dyn_hatTassle_1_r, swing1LimitDegrees=20, swing2LimitDegrees=8, twistLimitDegrees=8}
 {bone1Name=dyn_hatTassle_3_r, bone2Name=dyn_hatTassle_2_r, swing1LimitDegrees=20, swing2LimitDegrees=8, twistLimitDegrees=8}
 {
  bone1Name=dyn_hatTassle_5_Base_l
  bone2Name=neck_02
  swing1LimitDegrees=45
  swing1Motion=Free
  swing2LimitDegrees=45
  swing2Motion=Free
  twistLimitDegrees=45
  twistMotion=Free
 }
 {bone1Name=dyn_hatTassle_5_l, bone2Name=dyn_hatTassle_5_Base_l, swing1LimitDegrees=20, swing2LimitDegrees=8, twistLimitDegrees=8}
 {bone1Name=dyn_hatTassle_6_l, bone2Name=dyn_hatTassle_5_l, swing1LimitDegrees=20, swing2LimitDegrees=8, twistLimitDegrees=8}
 {bone1Name=dyn_hatTassle_7_l, bone2Name=dyn_hatTassle_6_l, swing1LimitDegrees=20, swing2LimitDegrees=8, twistLimitDegrees=8}
 {
  bone1Name=dyn_hatFlap_01_Base_r
  bone2Name=neck_02
  swing1LimitDegrees=45
  swing1Motion=Free
  swing2LimitDegrees=45
  swing2Motion=Free
  twistLimitDegrees=45
  twistMotion=Free
 }
 {bone1Name=dyn_hatFlap_01_r, bone2Name=dyn_hatFlap_01_Base_r, swing1LimitDegrees=18, swing2LimitDegrees=16, twistLimitDegrees=8}
 {bone1Name=dyn_hatFlap_02_r, bone2Name=dyn_hatFlap_01_r, swing1LimitDegrees=18, swing2LimitDegrees=16, twistLimitDegrees=8}
 {
  bone1Name=dyn_hatFlap_01_Base_l
  bone2Name=neck_02
  swing1LimitDegrees=45
  swing1Motion=Free
  swing2LimitDegrees=45
  swing2Motion=Free
  twistLimitDegrees=45
  twistMotion=Free
 }
 {bone1Name=dyn_hatFlap_01_l, bone2Name=dyn_hatFlap_01_Base_l, swing1LimitDegrees=18, swing2LimitDegrees=16, twistLimitDegrees=8}
 {bone1Name=dyn_hatFlap_02_l, bone2Name=dyn_hatFlap_01_l, swing1LimitDegrees=18, swing2LimitDegrees=16, twistLimitDegrees=8}
 {bone1Name=dyn_hattail_01, bone2Name=neck_02, swing1LimitDegrees=14, swing2LimitDegrees=10, twistLimitDegrees=10}
 {bone1Name=dyn_hattail_02, bone2Name=dyn_hattail_01, swing1LimitDegrees=16, swing2LimitDegrees=12, twistLimitDegrees=12}
]
